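# Sound cue for the doorbell "poof" effect, filed under Athena quest doorbell sounds.
# It plays a single wave, Doorbell_Poof_01, through the ambient props sound class at slightly reduced volume.
# Overridden attenuation gives a short-range falloff with occlusion low-pass filtering.
# Overridden concurrency allows one instance at a time, stopping the oldest when a new one starts.
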SoundCue /Game/Athena/Sounds/Quests/Doorbells/Doorbell_Poof_Cue
root=WavePlayer_2
nodes [{id=WavePlayer_2, properties {soundWaveAssetPtr=&/Doorbell_Poof_01}}]
properties {
 attenuationSettings=/Game/Sounds/Attenuation/Fort_Prop_Close_Att
 bOverrideAttenuation=true
 bOverrideConcurrency=true
 maxDistance=1536
 soundClassObject=&Fort_Ambient_Props_Athena
 volumeMultiplier=0.928
 attenuationOverrides {
  bEnableOcclusion=true
  falloffDistance=1536
  occlusionLowPassFilterFrequency=3000
  occlusionTraceChannel=ECC_GameTraceChannel11
  stereoSpread=256
  attenuationShapeExtents {x=0}
 }
 concurrencyOverrides {maxCount=1, resolutionRule=StopOldest}
}
