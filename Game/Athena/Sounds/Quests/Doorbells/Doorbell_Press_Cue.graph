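# Sound cue for a doorbell button press click, filed under Athena quest sounds and played as a quiet ambient prop sound.
# A quality-level switch sends the first two levels to an evenly weighted random pick of four click waves, and the last two to a single click wave.
# Overrides give it short-range attenuation with low-pass occlusion and a small concurrency cap that stops the oldest instance.
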SoundCue /Game/Athena/Sounds/Quests/Doorbells/Doorbell_Press_Cue
root=QualityLevel_0
nodes [
 {id=QualityLevel_0, children [Random_0,Random_0,WavePlayer_4,WavePlayer_4]}
 {id=Random_0, children [WavePlayer_0,WavePlayer_1,WavePlayer_2,WavePlayer_3], properties {weights [1,1,1,1]}}
 {id=WavePlayer_0, properties {soundWaveAssetPtr=&/Doorbell_Click_01}}
 {id=WavePlayer_1, properties {soundWaveAssetPtr=&/Doorbell_Click_03}}
 {id=WavePlayer_2, properties {soundWaveAssetPtr=&/Doorbell_Click_02}}
 {id=WavePlayer_3, properties {soundWaveAssetPtr=&/Doorbell_Click_04}}
 {id=WavePlayer_4, properties {soundWaveAssetPtr=&Doorbell_Click_01}}
]
properties {
 attenuationSettings=/Game/Sounds/Attenuation/Fort_Prop_Close_Att
 bOverrideAttenuation=true
 bOverrideConcurrency=true
 maxDistance=1536
 soundClassObject=&Fort_Ambient_Props_Athena
 volumeMultiplier=0.186
 attenuationOverrides {
  bEnableOcclusion=true
  falloffDistance=1536
  occlusionLowPassFilterFrequency=3000
  occlusionTraceChannel=ECC_GameTraceChannel11
  stereoSpread=256
  attenuationShapeExtents {x=0}
 }
 concurrencyOverrides {maxCount=3, resolutionRule=StopOldest}
}
