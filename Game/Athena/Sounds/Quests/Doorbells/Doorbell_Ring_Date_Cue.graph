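# Sound cue for the Hightower Date doorbell ring, a quest prop sound in Athena, assigned to the ambient props sound class.
# A mixer layers the main doorbell wave with a half-volume Far variant, each through its own attenuation node with overridden settings.
# Only the main layer enables occlusion, which low-passes and lowers its volume when the trace is blocked.
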SoundCue /Game/Athena/Sounds/Quests/Doorbells/Doorbell_Ring_Date_Cue
root=Mixer_0
nodes [
 {
  id=Attenuation_2
  children [WavePlayer_2]
  properties {
   bOverrideAttenuation=true
   attenuationOverrides {
    bEnableOcclusion=true
    occlusionInterpolationTime=1
    occlusionLowPassFilterFrequency=800
    occlusionTraceChannel=ECC_GameTraceChannel11
    occlusionVolumeAttenuation=0.2
   }
  }
 }
 {id=Attenuation_3, children [WavePlayer_3], properties {bOverrideAttenuation=true}}
 {id=Mixer_0, children [Attenuation_2, Attenuation_3], properties {inputVolume [1,0.5]}}
 {id=WavePlayer_2, properties {soundWaveAssetPtr=&/Hightower_Date_Doorbell}}
 {id=WavePlayer_3, properties {soundWaveAssetPtr=&/Hightower_Date_Doorbell_Far}}
]
properties {
 maxDistance=4000
 soundClassObject=&Fort_Ambient_Props_Athena
 volumeMultiplier=0.928
 attenuationOverrides {distanceAlgorithm=Logarithmic, stereoSpread=512}
 concurrencyOverrides {bLimitToOwner=true, maxCount=2, resolutionRule=StopOldest}
}
