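# Sound cue for a rundown doorbell ring, filed under Athena quest sounds and routed to the ambient props sound class.
# A mixer plays the Doorbell_1 "In" and "Out" waves together at equal volume, each through its own attenuation override node.
# Only the "In" layer enables occlusion, with a low-pass filter and slow interpolation, so obstructions muffle it separately from the "Out" layer.
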SoundCue /Game/Athena/Sounds/Quests/Doorbells/Doorbell_Ring_Rundown
root=Mixer_0
nodes [
 {
  id=Attenuation_2
  children [WavePlayer_0]
  properties {bOverrideAttenuation=true, attenuationOverrides {bEnableOcclusion=true, occlusionInterpolationTime=3, occlusionLowPassFilterFrequency=800}}
 }
 {id=Attenuation_3, children [WavePlayer_1], properties {bOverrideAttenuation=true}}
 {id=Mixer_0, children [Attenuation_2, Attenuation_3], properties {inputVolume [1,1]}}
 {id=WavePlayer_0, properties {soundWaveAssetPtr=&/Doorbell_1_In}}
 {id=WavePlayer_1, properties {soundWaveAssetPtr=&/Doorbell_1_Out}}
]
properties {
 maxDistance=4000
 soundClassObject=&Fort_Ambient_Props_Athena
 volumeMultiplier=0.928
 attenuationOverrides {distanceAlgorithm=Logarithmic, stereoSpread=512}
 concurrencyOverrides {bLimitToOwner=true, maxCount=2, resolutionRule=StopOldest}
}
